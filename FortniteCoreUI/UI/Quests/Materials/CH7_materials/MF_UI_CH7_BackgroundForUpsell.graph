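# Chapter 7 UI material function that renders an animated, tropical background for upsell screens in the quests UI.
# A curve-atlas radial gradient blends with warm tints to form the base color, driven by VaultWorld UV helpers and aspect-ratio scaling.
# Mirrored, twisted, rotating palm layers darken it through Color Burn.
# Time-driven texture effects, with perspective distortion and chromatic aberration, feed a final Color Dodge blend.
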MaterialFunction /FortniteCoreUI/UI/Quests/Materials/CH7_materials/MF_UI_CH7_BackgroundForUpsell
graph {
 expressions {
  MaterialExpressionAdd_11 {
   class=MaterialExpressionAdd
   inputs [
    A {from=MaterialExpressionNamedRerouteUsage_43}
    B {from=MaterialExpressionAppendVector_3}
   ]
   properties {constB=0}
  }
  MaterialExpressionAdd_12 {
   class=MaterialExpressionAdd
   inputs [
    A {from=MaterialExpressionMultiply_84}
    B {from=""}
   ]
   properties {constB=0.5}
  }
  MaterialExpressionAdd_13 {
   class=MaterialExpressionAdd
   inputs [
    A {from=MaterialExpressionScalarParameter_23}
    B {from=""}
   ]
   properties {constB=0.05}
  }
  MaterialExpressionAdd_14 {
   class=MaterialExpressionAdd
   inputs [
    A {from=MaterialExpressionConstant2Vector_22}
    B {from=MaterialExpressionAdd_41}
   ]
  }
  MaterialExpressionAdd_17 {
   class=MaterialExpressionAdd
   inputs [
    A {from=MaterialExpressionMultiply_28}
    B {from=""}
   ]
   properties {constB=0.5}
  }
  MaterialExpressionAdd_2 {
   class=MaterialExpressionAdd
   inputs [
    A {from=MaterialExpressionNamedRerouteUsage_30}
    B {from=MaterialExpressionConstant2Vector_10}
   ]
  }
  MaterialExpressionAdd_32 {
   class=MaterialExpressionAdd
   inputs [
    A {from=MaterialExpressionLinearInterpolate_85}
    B {from=MaterialExpressionRotator_14}
   ]
  }
  MaterialExpressionAdd_33 {
   class=MaterialExpressionAdd
   inputs [
    A {from=MaterialExpressionMaterialFunctionCall_39, output=Symmetry}
    B {from=MaterialExpressionLinearInterpolate_106}
   ]
  }
  MaterialExpressionAdd_34 {
   class=MaterialExpressionAdd
   inputs [
    A {from=MaterialExpressionLinearInterpolate_88}
    B {from=MaterialExpressionRotator_16}
   ]
  }
  MaterialExpressionAdd_35 {
   class=MaterialExpressionAdd
   inputs [
    A {from=MaterialExpressionMaterialFunctionCall_39, output=Symmetry}
    B {from=MaterialExpressionLinearInterpolate_108}
   ]
  }
  MaterialExpressionAdd_36 {
   class=MaterialExpressionAdd
   inputs [
    A {from=MaterialExpressionLinearInterpolate_93}
    B {from=MaterialExpressionRotator_18}
   ]
  }
  MaterialExpressionAdd_37 {
   class=MaterialExpressionAdd
   inputs [
    A {from=MaterialExpressionNamedRerouteDeclaration_77}
    B {from=MaterialExpressionConstant2Vector_47}
   ]
  }
  MaterialExpressionAdd_38 {
   class=MaterialExpressionAdd
   inputs [
    A {from=MaterialExpressionMultiply_68}
    B {from=MaterialExpressionConstant2Vector_49}
   ]
  }
  MaterialExpressionAdd_39 {
   class=MaterialExpressionAdd
   inputs [
    A {from=MaterialExpressionMultiply_71}
    B {from=""}
   ]
   properties {constB=0.5}
  }
  MaterialExpressionAdd_40 {
   class=MaterialExpressionAdd
   inputs [
    A {from=MaterialExpressionRotator_20}
    B {from=MaterialExpressionAppendVector_7}
   ]
  }
  MaterialExpressionAdd_41 {
   class=MaterialExpressionAdd
   inputs [
    A {from=MaterialExpressionMultiply_76}
    B {from=MaterialExpressionMaterialFunctionCall_46, output=Twist UV}
   ]
  }
  MaterialExpressionAdd_42 {
   class=MaterialExpressionAdd
   inputs [
    A {from=MaterialExpressionNamedRerouteUsage_121}
    B {from=MaterialExpressionConstant2Vector_57}
   ]
  }
  MaterialExpressionAdd_43 {
   class=MaterialExpressionAdd
   inputs [
    A {from=MaterialExpressionMultiply_66}
    B {from=MaterialExpressionConstant2Vector_58}
   ]
  }
  MaterialExpressionAdd_44 {
   class=MaterialExpressionAdd
   inputs [
    A {from=MaterialExpressionMultiply_80}
    B {from=""}
   ]
   properties {constB=0.5}
  }
  MaterialExpressionAppendVector_2 {
   class=MaterialExpressionAppendVector
   inputs [
    A {from=MaterialExpressionMultiply_3, output=""}
    B {from=MaterialExpressionConstant_4, output=""}
   ]
  }
  MaterialExpressionAppendVector_3 {
   class=MaterialExpressionAppendVector
   inputs [
    A {from=MaterialExpressionMultiply_30, output=""}
    B {from=MaterialExpressionLinearInterpolate_45, output=""}
   ]
  }
  MaterialExpressionAppendVector_5 {
   class=MaterialExpressionAppendVector
   inputs [
    A {from=MaterialExpressionDivide_6, output=""}
    B {from=MaterialExpressionConstant_22, output=""}
   ]
  }
  MaterialExpressionAppendVector_7 {
   class=MaterialExpressionAppendVector
   inputs [
    A {from=MaterialExpressionNamedRerouteDeclaration_82, output=""}
    B {from=MaterialExpressionConstant_48, output=""}
   ]
  }
  MaterialExpressionAppendVector_8 {
   class=MaterialExpressionAppendVector
   inputs [
    A {from=MaterialExpressionConstant_52, output=""}
    B {from=MaterialExpressionConstant_53, output=""}
   ]
  }
  MaterialExpressionComponentMask_11 {class=MaterialExpressionComponentMask, inputs [None {from=MaterialExpressionNamedRerouteUsage_3}], properties {r=true}}
  MaterialExpressionComponentMask_12 {class=MaterialExpressionComponentMask, inputs [None {from=MaterialExpressionSubtract_2}], properties {r=true}}
  MaterialExpressionComponentMask_16 {class=MaterialExpressionComponentMask, inputs [None {from=MaterialExpressionNamedRerouteUsage_3}], properties {g=true}}
  MaterialExpressionComponentMask_17 {class=MaterialExpressionComponentMask, inputs [None {from=MaterialExpressionMaterialFunctionCall_22, output=H Ratio}], properties {r=true}}
  MaterialExpressionComponentMask_28 {class=MaterialExpressionComponentMask, inputs [None {from=MaterialExpressionNamedRerouteUsage_105}], properties {g=true}}
  MaterialExpressionComponentMask_29 {class=MaterialExpressionComponentMask, inputs [None {from=MaterialExpressionNamedRerouteDeclaration_68}], properties {g=true}}
  MaterialExpressionComponentMask_30 {class=MaterialExpressionComponentMask, inputs [None {from=MaterialExpressionNamedRerouteUsage_109}], properties {g=true}}
  MaterialExpressionComponentMask_31 {class=MaterialExpressionComponentMask, inputs [None {from=MaterialExpressionNamedRerouteDeclaration_71}], properties {g=true}}
  MaterialExpressionComponentMask_32 {class=MaterialExpressionComponentMask, inputs [None {from=MaterialExpressionMaterialFunctionCall_38, output=UV}], properties {r=true}}
  MaterialExpressionComponentMask_33 {class=MaterialExpressionComponentMask, inputs [None {from=MaterialExpressionNamedRerouteUsage_115}], properties {g=true}}
  MaterialExpressionComponentMask_34 {class=MaterialExpressionComponentMask, inputs [None {from=MaterialExpressionNamedRerouteDeclaration_76}], properties {g=true}}
  MaterialExpressionComponentMask_35 {class=MaterialExpressionComponentMask, inputs [None {from=MaterialExpressionAdd_42}], properties {g=true}}
  MaterialExpressionComponentMask_36 {class=MaterialExpressionComponentMask, inputs [None {from=MaterialExpressionNamedRerouteUsage_60}], properties {g=true}}
  MaterialExpressionComponentMask_37 {class=MaterialExpressionComponentMask, inputs [None {from=MaterialExpressionNamedRerouteUsage_138}], properties {r=true}}
  MaterialExpressionConstant2Vector_10 {class=MaterialExpressionConstant2Vector, properties {g=-0.2}}
  MaterialExpressionConstant2Vector_11 {class=MaterialExpressionConstant2Vector, properties {r=0.5}}
  MaterialExpressionConstant2Vector_15 {class=MaterialExpressionConstant2Vector, properties {g=2, r=-1}}
  MaterialExpressionConstant2Vector_16 {class=MaterialExpressionConstant2Vector, properties {g=-1, r=-2}}
  MaterialExpressionConstant2Vector_17 {class=MaterialExpressionConstant2Vector, properties {g=1, r=0.75}}
  MaterialExpressionConstant2Vector_22 {class=MaterialExpressionConstant2Vector, properties {r=0.1}}
  MaterialExpressionConstant2Vector_43 {class=MaterialExpressionConstant2Vector, properties {g=-0.05, r=2.8}}
  MaterialExpressionConstant2Vector_44 {class=MaterialExpressionConstant2Vector, properties {g=1, r=0.5}}
  MaterialExpressionConstant2Vector_45 {class=MaterialExpressionConstant2Vector, properties {r=1.6}}
  MaterialExpressionConstant2Vector_46 {class=MaterialExpressionConstant2Vector, properties {g=1.8, r=1.2}}
  MaterialExpressionConstant2Vector_47 {class=MaterialExpressionConstant2Vector, properties {r=-1.8}}
  MaterialExpressionConstant2Vector_48 {class=MaterialExpressionConstant2Vector, properties {g=1.2, r=1.2}}
  MaterialExpressionConstant2Vector_49 {class=MaterialExpressionConstant2Vector, properties {g=1, r=0.5}}
  MaterialExpressionConstant2Vector_50 {class=MaterialExpressionConstant2Vector, properties {g=2}}
  MaterialExpressionConstant2Vector_51 {class=MaterialExpressionConstant2Vector, properties {g=3, r=2}}
  MaterialExpressionConstant2Vector_52 {class=MaterialExpressionConstant2Vector, properties {g=-2, r=2}}
  MaterialExpressionConstant2Vector_53 {class=MaterialExpressionConstant2Vector, properties {g=0.2, r=2.8}}
  MaterialExpressionConstant2Vector_54 {class=MaterialExpressionConstant2Vector, properties {g=0.9, r=0.45}}
  MaterialExpressionConstant2Vector_55 {class=MaterialExpressionConstant2Vector, properties {g=-0.075, r=1.9}}
  MaterialExpressionConstant2Vector_56 {class=MaterialExpressionConstant2Vector, properties {g=1.26, r=0.7193}}
  MaterialExpressionConstant2Vector_57 {class=MaterialExpressionConstant2Vector, properties {g=0.1}}
  MaterialExpressionConstant2Vector_58 {class=MaterialExpressionConstant2Vector, properties {g=1, r=0.5}}
  MaterialExpressionConstant2Vector_9 {class=MaterialExpressionConstant2Vector, properties {g=0.5, r=0.8}}
  MaterialExpressionConstant3Vector_0 {class=MaterialExpressionConstant3Vector, properties {constant=<rgba 0.7913 0.1946 0.0529 1>}}
  MaterialExpressionConstant3Vector_4 {class=MaterialExpressionConstant3Vector, properties {constant=<rgba 0.0685 0.1195 0.4452 1>}}
  MaterialExpressionConstant4Vector_0 {class=MaterialExpressionConstant4Vector, inputs [], properties {constant=<rgba 0.8837 0.6879 0.1447 0.7>}}
  MaterialExpressionConstantBiasScale_4 {class=MaterialExpressionConstantBiasScale, inputs [None {from=MaterialExpressionSine_4}]}
  MaterialExpressionConstant_15 {class=MaterialExpressionConstant, properties {r=0.8}}
  MaterialExpressionConstant_18 {class=MaterialExpressionConstant, properties {r=0.45}}
  MaterialExpressionConstant_22 {class=MaterialExpressionConstant, properties {r=0.75}}
  MaterialExpressionConstant_24 {class=MaterialExpressionConstant, properties {r=2.4}}
  MaterialExpressionConstant_25 {class=MaterialExpressionConstant, properties {r=0.5}}
  MaterialExpressionConstant_26 {class=MaterialExpressionConstant, properties {r=0.7}}
  MaterialExpressionConstant_34 {class=MaterialExpressionConstant, properties {r=0.01}}
  MaterialExpressionConstant_4 {class=MaterialExpressionConstant}
  MaterialExpressionConstant_40 {class=MaterialExpressionConstant, properties {r=0.01}}
  MaterialExpressionConstant_41 {class=MaterialExpressionConstant, properties {r=-0.025}}
  MaterialExpressionConstant_42 {class=MaterialExpressionConstant, properties {r=0.01}}
  MaterialExpressionConstant_43 {class=MaterialExpressionConstant, properties {r=-0.025}}
  MaterialExpressionConstant_44 {class=MaterialExpressionConstant, properties {r=0.01}}
  MaterialExpressionConstant_45 {class=MaterialExpressionConstant, properties {r=0.0125}}
  MaterialExpressionConstant_46 {class=MaterialExpressionConstant, properties {r=-0.175}}
  MaterialExpressionConstant_47 {class=MaterialExpressionConstant, properties {r=3}}
  MaterialExpressionConstant_48 {class=MaterialExpressionConstant}
  MaterialExpressionConstant_49 {class=MaterialExpressionConstant, properties {r=4}}
  MaterialExpressionConstant_5 {class=MaterialExpressionConstant, properties {r=0.5}}
  MaterialExpressionConstant_50 {class=MaterialExpressionConstant, properties {r=0.1}}
  MaterialExpressionConstant_51 {class=MaterialExpressionConstant, properties {r=1}}
  MaterialExpressionConstant_52 {class=MaterialExpressionConstant, properties {r=1.8}}
  MaterialExpressionConstant_53 {class=MaterialExpressionConstant, properties {r=1}}
  MaterialExpressionConstant_6 {class=MaterialExpressionConstant, properties {r=-1}}
  MaterialExpressionConstant_7 {class=MaterialExpressionConstant, properties {r=0.7778}}
  MaterialExpressionConvert_1 {
   class=MaterialExpressionConvert
   inputs [
    None {from=MaterialExpressionLinearInterpolate_47}
    None {from=MaterialExpressionTextureSample_20, output=R}
   ]
   properties {
    desc=Chromatic Aberration
    nodeName=Make Vector3
    convertInputs [
     {
      type=Scalar
      expressionInput {
       expression=&self:MaterialExpressionLinearInterpolate_47
       inputName=None
       mask=0
       maskA=0
       maskB=0
       maskG=0
       maskR=0
       outputIndex=0
      }
     }
     {
      type=Scalar
      expressionInput {
       expression=&self:MaterialExpressionTextureSample_20
       inputName=None
       mask=1
       maskA=0
       maskB=0
       maskG=0
       maskR=1
       outputIndex=1
      }
     }
    ]
    convertOutputs [{type=Vector3}]
   }
  }
  MaterialExpressionCurveAtlasRowParameter_0 {
   class=MaterialExpressionCurveAtlasRowParameter
   inputs [CurveTime {from=MaterialExpressionMaterialFunctionCall_21, output=Linear}]
   properties {
    atlas=&/T_UI_Ch7S1_CurveAtlasQuests
    curve=&/T_UI_Ch7S1_ColorCurve
    parameterName=BackgroundCurve
   }
  }
  MaterialExpressionDivide_5 {
   class=MaterialExpressionDivide
   inputs [
    A {from=MaterialExpressionMaterialFunctionCall_48, output=Result}
    B {from=MaterialExpressionLinearInterpolate_82}
   ]
   properties {desc=Palm Vibrancy}
  }
  MaterialExpressionDivide_6 {
   class=MaterialExpressionDivide
   inputs [
    A {from=MaterialExpressionConstant_25}
    B {from=MaterialExpressionMax_3}
   ]
  }
  MaterialExpressionFrac_3 {class=MaterialExpressionFrac, inputs [None {from=MaterialExpressionMultiply_72, output=""}], properties {}}
  MaterialExpressionFrac_4 {class=MaterialExpressionFrac, inputs [None {from=MaterialExpressionMultiply_41, output=""}], properties {}}
  MaterialExpressionFunctionOutput_1 {class=MaterialExpressionFunctionOutput, inputs [None {from=MaterialExpressionMaterialFunctionCall_45, output=Result}]}
  MaterialExpressionLength_0 {class=MaterialExpressionLength, inputs [None {from=MaterialExpressionSubtract_6, output=""}], properties {}}
  MaterialExpressionLength_2 {class=MaterialExpressionLength, inputs [None {from=MaterialExpressionMultiply_26, output=""}], properties {}}
  MaterialExpressionLength_4 {class=MaterialExpressionLength, inputs [None {from=MaterialExpressionSubtract_19, output=""}], properties {}}
  MaterialExpressionLinearInterpolate_100 {
   class=MaterialExpressionLinearInterpolate
   inputs [
    A {from=""}
    B {from=""}
    Alpha {from=MaterialExpressionSmoothStep_5}
   ]
   properties {constB=12}
  }
  MaterialExpressionLinearInterpolate_101 {
   class=MaterialExpressionLinearInterpolate
   inputs [
    A {from=""}
    B {from=MaterialExpressionConvert_1}
    Alpha {from=MaterialExpressionNamedRerouteDeclaration_80}
   ]
   properties {constA=1, constB=64}
  }
  MaterialExpressionLinearInterpolate_102 {
   class=MaterialExpressionLinearInterpolate
   inputs [
    A {from=""}
    B {from=MaterialExpressionLinearInterpolate_111}
    Alpha {from=MaterialExpressionLinearInterpolate_95}
   ]
   properties {constA=1, constB=0.9}
  }
  MaterialExpressionLinearInterpolate_103 {
   class=MaterialExpressionLinearInterpolate
   inputs [
    A {from=MaterialExpressionScalarParameter_18}
    B {from=MaterialExpressionMultiply_79}
    Alpha {from=MaterialExpressionSaturate_13}
   ]
  }
  MaterialExpressionLinearInterpolate_104 {
   class=MaterialExpressionLinearInterpolate
   inputs [
    A {from=""}
    B {from=""}
    Alpha {from=MaterialExpressionNamedRerouteDeclaration_82}
   ]
   properties {constA=0.001, constB=-0.001}
  }
  MaterialExpressionLinearInterpolate_105 {
   class=MaterialExpressionLinearInterpolate
   inputs [
    A {from=""}
    B {from=""}
    Alpha {from=MaterialExpressionConstantBiasScale_4}
   ]
  }
  MaterialExpressionLinearInterpolate_106 {
   class=MaterialExpressionLinearInterpolate
   inputs [
    A {from=MaterialExpressionConstant2Vector_43}
    B {from=MaterialExpressionConstant2Vector_53}
    Alpha {from=MaterialExpressionNamedRerouteUsage_133}
   ]
  }
  MaterialExpressionLinearInterpolate_107 {
   class=MaterialExpressionLinearInterpolate
   inputs [
    A {from=MaterialExpressionConstant2Vector_44}
    B {from=MaterialExpressionConstant2Vector_54}
    Alpha {from=MaterialExpressionNamedRerouteUsage_134}
   ]
  }
  MaterialExpressionLinearInterpolate_108 {
   class=MaterialExpressionLinearInterpolate
   inputs [
    A {from=MaterialExpressionConstant2Vector_45}
    B {from=MaterialExpressionConstant2Vector_55}
    Alpha {from=MaterialExpressionNamedRerouteUsage_135}
   ]
  }
  MaterialExpressionLinearInterpolate_109 {
   class=MaterialExpressionLinearInterpolate
   inputs [
    A {from=MaterialExpressionConstant2Vector_46}
    B {from=MaterialExpressionConstant2Vector_56}
    Alpha {from=MaterialExpressionNamedRerouteUsage_136}
   ]
   properties {constB=1.15}
  }
  MaterialExpressionLinearInterpolate_110 {
   class=MaterialExpressionLinearInterpolate
   inputs [
    A {from=""}
    B {from=""}
    Alpha {from=MaterialExpressionNamedRerouteUsage_137}
   ]
   properties {constA=-0.1, constB=0}
  }
  MaterialExpressionLinearInterpolate_111 {
   class=MaterialExpressionLinearInterpolate
   inputs [
    A {from=MaterialExpressionScalarParameter_23}
    B {from=MaterialExpressionMin_0}
    Alpha {from=MaterialExpressionComponentMask_37}
   ]
   properties {constA=0.9, constB=0.95, desc=Color Burn Intensity}
  }
  MaterialExpressionLinearInterpolate_112 {
   class=MaterialExpressionLinearInterpolate
   inputs [
    A {from=MaterialExpressionLinearInterpolate_113}
    B {from=MaterialExpressionScalarParameter_19}
    Alpha {from=MaterialExpressionScalarParameter_11}
   ]
  }
  MaterialExpressionLinearInterpolate_113 {
   class=MaterialExpressionLinearInterpolate
   inputs [
    A {from=MaterialExpressionLinearInterpolate_114}
    B {from=""}
    Alpha {from=MaterialExpressionNamedRerouteUsage_139}
   ]
   properties {constB=1.175}
  }
  MaterialExpressionLinearInterpolate_114 {
   class=MaterialExpressionLinearInterpolate
   inputs [
    A {from=""}
    B {from=""}
    Alpha {from=""}
   ]
   properties {constA=1.25}
  }
  MaterialExpressionLinearInterpolate_19 {
   class=MaterialExpressionLinearInterpolate
   inputs [
    A {from=""}
    B {from=""}
    Alpha {from=MaterialExpressionNamedRerouteDeclaration_39}
   ]
   properties {constA=-1}
  }
  MaterialExpressionLinearInterpolate_2 {
   class=MaterialExpressionLinearInterpolate
   inputs [
    A {from=MaterialExpressionLinearInterpolate_43}
    B {from=MaterialExpressionMultiply_25}
    Alpha {from=MaterialExpressionNamedRerouteDeclaration_41}
   ]
  }
  MaterialExpressionLinearInterpolate_31 {
   class=MaterialExpressionLinearInterpolate
   inputs [
    A {from=MaterialExpressionLinearInterpolate_37}
    B {from=MaterialExpressionConstant4Vector_0, output=RGB}
    Alpha {from=MaterialExpressionMultiply_27}
   ]
  }
  MaterialExpressionLinearInterpolate_34 {
   class=MaterialExpressionLinearInterpolate
   inputs [
    A {from=MaterialExpressionNamedRerouteDeclaration_42}
    B {from=MaterialExpressionNamedRerouteUsage_7}
    Alpha {from=MaterialExpressionLinearInterpolate_36}
   ]
  }
  MaterialExpressionLinearInterpolate_36 {
   class=MaterialExpressionLinearInterpolate
   inputs [
    A {from=""}
    B {from=MaterialExpressionConstant_15}
    Alpha {from=MaterialExpressionSmoothStep_3}
   ]
   properties {constB=0.5}
  }
  MaterialExpressionLinearInterpolate_37 {
   class=MaterialExpressionLinearInterpolate
   inputs [
    A {from=MaterialExpressionNamedRerouteUsage_38}
    B {from=MaterialExpressionLinearInterpolate_34}
    Alpha {from=MaterialExpressionNamedRerouteUsage_42}
   ]
  }
  MaterialExpressionLinearInterpolate_42 {
   class=MaterialExpressionLinearInterpolate
   inputs [
    A {from=""}
    B {from=""}
    Alpha {from=MaterialExpressionScalarParameter_10}
   ]
   properties {constB=0}
  }
  MaterialExpressionLinearInterpolate_43 {
   class=MaterialExpressionLinearInterpolate
   inputs [
    A {from=MaterialExpressionCurveAtlasRowParameter_0}
    B {from=MaterialExpressionConstant3Vector_0}
    Alpha {from=MaterialExpressionComponentMask_16}
   ]
  }
  MaterialExpressionLinearInterpolate_44 {
   class=MaterialExpressionLinearInterpolate
   inputs [
    A {from=MaterialExpressionViewSize_0}
    B {from=MaterialExpressionMaterialFunctionCall_26, output=Pixel Size}
    Alpha {from=MaterialExpressionScalarParameter_8}
   ]
  }
  MaterialExpressionLinearInterpolate_45 {
   class=MaterialExpressionLinearInterpolate
   inputs [
    A {from=""}
    B {from=""}
    Alpha {from=MaterialExpressionNamedRerouteDeclaration_44}
   ]
   properties {constA=0.5, constB=0.6}
  }
  MaterialExpressionLinearInterpolate_47 {
   class=MaterialExpressionLinearInterpolate
   inputs [
    A {from=""}
    B {from=""}
    Alpha {from=MaterialExpressionComponentMask_36}
   ]
   properties {constB=512, desc=Color Burn Intensity}
  }
  MaterialExpressionLinearInterpolate_8 {
   class=MaterialExpressionLinearInterpolate
   inputs [
    A {from=MaterialExpressionConstant3Vector_0}
    B {from=MaterialExpressionReroute_0}
    Alpha {from=MaterialExpressionComponentMask_11}
   ]
  }
  MaterialExpressionLinearInterpolate_82 {
   class=MaterialExpressionLinearInterpolate
   inputs [
    A {from=""}
    B {from=""}
    Alpha {from=MaterialExpressionLinearInterpolate_94}
   ]
   properties {constA=1, constB=0.8}
  }
  MaterialExpressionLinearInterpolate_83 {
   class=MaterialExpressionLinearInterpolate
   inputs [
    A {from=MaterialExpressionConstant_40}
    B {from=MaterialExpressionMultiply_63}
    Alpha {from=MaterialExpressionNamedRerouteUsage_129}
   ]
   properties {constA=0.05, constB=-0.05}
  }
  MaterialExpressionLinearInterpolate_84 {
   class=MaterialExpressionLinearInterpolate
   inputs [
    A {from=MaterialExpressionLinearInterpolate_83}
    B {from=""}
    Alpha {from=MaterialExpressionComponentMask_28}
   ]
   properties {constB=0}
  }
  MaterialExpressionLinearInterpolate_85 {
   class=MaterialExpressionLinearInterpolate
   inputs [
    A {from=MaterialExpressionNamedRerouteUsage_104}
    B {from=""}
    Alpha {from=MaterialExpressionComponentMask_29}
   ]
   properties {constB=0}
  }
  MaterialExpressionLinearInterpolate_86 {
   class=MaterialExpressionLinearInterpolate
   inputs [
    A {from=MaterialExpressionConstant_42}
    B {from=MaterialExpressionMultiply_65}
    Alpha {from=MaterialExpressionNamedRerouteUsage_130}
   ]
   properties {constA=0.05, constB=-0.05}
  }
  MaterialExpressionLinearInterpolate_87 {
   class=MaterialExpressionLinearInterpolate
   inputs [
    A {from=MaterialExpressionLinearInterpolate_86}
    B {from=""}
    Alpha {from=MaterialExpressionComponentMask_30}
   ]
   properties {constB=0}
  }
  MaterialExpressionLinearInterpolate_88 {
   class=MaterialExpressionLinearInterpolate
   inputs [
    A {from=MaterialExpressionNamedRerouteUsage_108}
    B {from=""}
    Alpha {from=MaterialExpressionComponentMask_31}
   ]
   properties {constB=0}
  }
  MaterialExpressionLinearInterpolate_89 {
   class=MaterialExpressionLinearInterpolate
   inputs [
    A {from=MaterialExpressionNamedRerouteUsage_102}
    B {from=""}
    Alpha {from=MaterialExpressionNamedRerouteUsage_113}
   ]
  }
  MaterialExpressionLinearInterpolate_90 {
   class=MaterialExpressionLinearInterpolate
   inputs [
    A {from=MaterialExpressionNamedRerouteUsage_103}
    B {from=""}
    Alpha {from=MaterialExpressionNamedRerouteUsage_112}
   ]
  }
  MaterialExpressionLinearInterpolate_91 {
   class=MaterialExpressionLinearInterpolate
   inputs [
    A {from=MaterialExpressionConstant_44}
    B {from=MaterialExpressionMultiply_67}
    Alpha {from=MaterialExpressionNamedRerouteUsage_131}
   ]
   properties {constA=0.05, constB=-0.05}
  }
  MaterialExpressionLinearInterpolate_92 {
   class=MaterialExpressionLinearInterpolate
   inputs [
    A {from=MaterialExpressionLinearInterpolate_91}
    B {from=""}
    Alpha {from=MaterialExpressionComponentMask_33}
   ]
   properties {constB=0}
  }
  MaterialExpressionLinearInterpolate_93 {
   class=MaterialExpressionLinearInterpolate
   inputs [
    A {from=MaterialExpressionNamedRerouteUsage_114}
    B {from=""}
    Alpha {from=MaterialExpressionComponentMask_34}
   ]
   properties {constB=0}
  }
  MaterialExpressionLinearInterpolate_94 {
   class=MaterialExpressionLinearInterpolate
   inputs [
    A {from=MaterialExpressionLinearInterpolate_90}
    B {from=""}
    Alpha {from=MaterialExpressionNamedRerouteUsage_118}
   ]
  }
  MaterialExpressionLinearInterpolate_95 {
   class=MaterialExpressionLinearInterpolate
   inputs [
    A {from=MaterialExpressionLinearInterpolate_89}
    B {from=""}
    Alpha {from=MaterialExpressionNamedRerouteUsage_119}
   ]
  }
  MaterialExpressionLinearInterpolate_96 {
   class=MaterialExpressionLinearInterpolate
   inputs [
    A {from=""}
    B {from=MaterialExpressionScalarParameter_22}
    Alpha {from=MaterialExpressionLinearInterpolate_95}
   ]
   properties {constA=1, constB=0.9}
  }
  MaterialExpressionLinearInterpolate_97 {
   class=MaterialExpressionLinearInterpolate
   inputs [
    A {from=MaterialExpressionDivide_5}
    B {from=MaterialExpressionNamedRerouteUsage_120}
    Alpha {from=MaterialExpressionPower_4}
   ]
  }
  MaterialExpressionLinearInterpolate_98 {
   class=MaterialExpressionLinearInterpolate
   inputs [
    A {from=""}
    B {from=MaterialExpressionScalarParameter_9}
    Alpha {from=MaterialExpressionPower_5}
   ]
   properties {constA=1, constB=0.4}
  }
  MaterialExpressionLinearInterpolate_99 {
   class=MaterialExpressionLinearInterpolate
   inputs [
    A {from=""}
    B {from=""}
    Alpha {from=MaterialExpressionSaturate_11}
   ]
   properties {constA=1, constB=2}
  }
  MaterialExpressionMaterialFunctionCall_20 {
   class=MaterialExpressionMaterialFunctionCall
   inputs ["Value (0-1)" {from=MaterialExpressionSaturate_0}]
   properties {materialFunction=&MF_UI_EaseCurves}
  }
  MaterialExpressionMaterialFunctionCall_21 {
   class=MaterialExpressionMaterialFunctionCall
   inputs [
    UV {from=MaterialExpressionAdd_17}
    GradientStart {from=""}
    GradientEnd {from=""}
   ]
   properties {materialFunction=&MF_UI_RadialGradient}
  }
  MaterialExpressionMaterialFunctionCall_22 {
   class=MaterialExpressionMaterialFunctionCall
   inputs [Dimensions {from=MaterialExpressionLinearInterpolate_44}]
   properties {materialFunction=/FortUILibrary/MaterialFunctions/Basics/MF_UI_GetAspectRatio}
  }
  MaterialExpressionMaterialFunctionCall_26 {class=MaterialExpressionMaterialFunctionCall, inputs [], properties {materialFunction=&GetUserInterfaceUV}}
  MaterialExpressionMaterialFunctionCall_3 {
   class=MaterialExpressionMaterialFunctionCall
   inputs [
    Use Local UV {from=""}
    Texture Size {from=""}
    Scale Pivot Point {from=""}
   ]
   properties {materialFunction=/Game/Environments/FrontEnd/Blueprints/ItemPreview/VaultWorldTransitions/MaterialFunctions/MF_UI_VaultWorld_UV}
  }
  MaterialExpressionMaterialFunctionCall_37 {
   class=MaterialExpressionMaterialFunctionCall
   inputs [
    UV {from=MaterialExpressionNamedRerouteUsage_100}
    Seed {from=""}
   ]
   properties {materialFunction=&MF_UI_Random}
  }
  MaterialExpressionMaterialFunctionCall_38 {
   class=MaterialExpressionMaterialFunctionCall
   inputs [
    Use Local UV {from=""}
    Texture Size {from=MaterialExpressionTextureProperty_2}
    Scale Pivot Point {from=""}
   ]
   properties {materialFunction=&MF_UI_VaultWorld_UV}
  }
  MaterialExpressionMaterialFunctionCall_39 {
   class=MaterialExpressionMaterialFunctionCall
   inputs [
    UV {from=MaterialExpressionNamedRerouteDeclaration_77}
    Mirror Y {from=""}
    Swizzle UV {from=""}
   ]
   properties {materialFunction=/FortUILibrary/MaterialFunctions/Basics/MF_UI_Symmetry_2x1_New}
  }
  MaterialExpressionMaterialFunctionCall_40 {
   class=MaterialExpressionMaterialFunctionCall
   inputs [
    UV {from=MaterialExpressionRotator_17}
    Twist Amount {from=MaterialExpressionLinearInterpolate_110}
   ]
   properties {materialFunction=/FortUILibrary/MaterialFunctions/Effects/MF_UI_Twist}
  }
  MaterialExpressionMaterialFunctionCall_41 {
   class=MaterialExpressionMaterialFunctionCall
   inputs [
    UV {from=MaterialExpressionRotator_15}
    Twist Amount {from=MaterialExpressionConstant_46}
   ]
   properties {materialFunction=&MF_UI_Twist}
  }
  MaterialExpressionMaterialFunctionCall_42 {
   class=MaterialExpressionMaterialFunctionCall
   inputs [
    UVs {from=MaterialExpressionMaterialFunctionCall_38, output=Scale Y}
    Texture Scale {from=MaterialExpressionLinearInterpolate_112}
   ]
   properties {materialFunction=&ScaleUVsByCenter}
  }
  MaterialExpressionMaterialFunctionCall_43 {
   class=MaterialExpressionMaterialFunctionCall
   inputs [
    UV {from=MaterialExpressionAdd_39}
    GradientStart {from=""}
    GradientEnd {from=""}
   ]
   properties {materialFunction=&MF_UI_RadialGradient}
  }
  MaterialExpressionMaterialFunctionCall_44 {
   class=MaterialExpressionMaterialFunctionCall
   inputs [
    "Value (0-1)" {from=MaterialExpressionFrac_3}
    Steps {from=MaterialExpressionConstant_49}
   ]
   properties {materialFunction=/FortUILibrary/MaterialFunctions/Basics/MF_UI_StepRatio}
  }
  MaterialExpressionMaterialFunctionCall_45 {
   class=MaterialExpressionMaterialFunctionCall
   inputs [
    Blend {from=MaterialExpressionNamedRerouteDeclaration_64}
    Base {from=MaterialExpressionNamedRerouteUsage_127}
   ]
   properties {materialFunction=/Engine/Functions/Engine_MaterialFunctions03/Blends/Blend_ColorDodge}
  }
  MaterialExpressionMaterialFunctionCall_46 {
   class=MaterialExpressionMaterialFunctionCall
   inputs [
    UV {from=MaterialExpressionMaterialFunctionCall_47, output=Symmetry}
    Twist Amount {from=MaterialExpressionConstant_50}
   ]
   properties {materialFunction=&MF_UI_Twist}
  }
  MaterialExpressionMaterialFunctionCall_47 {
   class=MaterialExpressionMaterialFunctionCall
   inputs [
    UV {from=MaterialExpressionMaterialFunctionCall_49, output=Perspective UVs}
    Mirror Y {from=""}
    Swizzle UV {from=""}
   ]
   properties {materialFunction=&MF_UI_Symmetry_2x1_New}
  }
  MaterialExpressionMaterialFunctionCall_48 {
   class=MaterialExpressionMaterialFunctionCall
   inputs [
    Base {from=MaterialExpressionMultiply_69}
    Blend {from=MaterialExpressionLinearInterpolate_102}
   ]
   properties {desc=Palm Burn, materialFunction=/Engine/Functions/Engine_MaterialFunctions03/Blends/Blend_ColorBurn}
  }
  MaterialExpressionMaterialFunctionCall_49 {
   class=MaterialExpressionMaterialFunctionCall
   inputs [
    UV {from=MaterialExpressionAdd_12}
    Perspective Rotation {from=MaterialExpressionLinearInterpolate_105}
    Perspective Amount {from=MaterialExpressionConstant_34}
   ]
   properties {materialFunction=/FortUILibrary/MaterialFunctions/Basics/MF_UI_Perspective}
  }
  MaterialExpressionMaterialFunctionCall_5 {
   class=MaterialExpressionMaterialFunctionCall
   inputs [
    Vault World Color {from=""}
    Focal Color Override {from=""}
    Focal Scale Override {from=MaterialExpressionScalarParameter_4}
    Mask Blend {from=""}
   ]
   properties {materialFunction=/Game/Environments/FrontEnd/Blueprints/ItemPreview/VaultWorldTransitions/MaterialFunctions/MF_UI_VaultWorld_FocalPoint}
  }
  MaterialExpressionMaterialFunctionCall_50 {
   class=MaterialExpressionMaterialFunctionCall
   inputs [
    1 {from=MaterialExpressionConstant2Vector_51}
    2 {from=MaterialExpressionConstant2Vector_52}
    3 {from=MaterialExpressionConstant2Vector_15}
    4 {from=MaterialExpressionConstant2Vector_16}
    A {from=MaterialExpressionNamedRerouteUsage_44}
   ]
   properties {materialFunction=/Engine/Functions/Engine_MaterialFunctions02/Lerp_Multiple_Float2}
  }
  MaterialExpressionMaterialFunctionCall_6 {
   class=MaterialExpressionMaterialFunctionCall
   inputs [
    UVs {from=MaterialExpressionNamedRerouteUsage_4}
    Texture Scale {from=MaterialExpressionConstant2Vector_9}
   ]
   properties {materialFunction=&ScaleUVsByCenter}
  }
  MaterialExpressionMax_2 {
   class=MaterialExpressionMax
   inputs [
    A {from=MaterialExpressionMultiply_31}
    B {from=""}
   ]
  }
  MaterialExpressionMax_3 {
   class=MaterialExpressionMax
   inputs [
    A {from=MaterialExpressionNamedRerouteUsage_141}
    B {from=""}
   ]
  }
  MaterialExpressionMax_6 {
   class=MaterialExpressionMax
   inputs [
    A {from=MaterialExpressionTextureSample_18, output=R}
    B {from=MaterialExpressionTextureSample_18, output=R}
   ]
  }
  MaterialExpressionMax_7 {
   class=MaterialExpressionMax
   inputs [
    A {from=MaterialExpressionMax_6}
    B {from=MaterialExpressionMultiply_73}
   ]
  }
  MaterialExpressionMin_0 {
   class=MaterialExpressionMin
   inputs [
    A {from=MaterialExpressionAdd_13, output=""}
    B {from="", output=""}
   ]
   properties {}
  }
  MaterialExpressionMultiply_2 {
   class=MaterialExpressionMultiply
   inputs [
    A {from=MaterialExpressionLinearInterpolate_19}
    B {from=MaterialExpressionMultiply_4}
   ]
   properties {constB=0.005}
  }
  MaterialExpressionMultiply_24 {
   class=MaterialExpressionMultiply
   inputs [
    A {from=MaterialExpressionSine_2}
    B {from=MaterialExpressionScalarParameter_7}
   ]
   properties {constB=0.25}
  }
  MaterialExpressionMultiply_25 {
   class=MaterialExpressionMultiply
   inputs [
    A {from=MaterialExpressionLinearInterpolate_43}
    B {from=""}
   ]
   properties {constB=3}
  }
  MaterialExpressionMultiply_26 {
   class=MaterialExpressionMultiply
   inputs [
    A {from=MaterialExpressionSubtract_8}
    B {from=MaterialExpressionConstant2Vector_17}
   ]
  }
  MaterialExpressionMultiply_27 {
   class=MaterialExpressionMultiply
   inputs [
    A {from=MaterialExpressionConstant4Vector_0, output=A}
    B {from=MaterialExpressionNamedRerouteUsage_9}
   ]
  }
  MaterialExpressionMultiply_28 {
   class=MaterialExpressionMultiply
   inputs [
    A {from=MaterialExpressionSubtract_9}
    B {from=""}
   ]
   properties {constB=0.275}
  }
  MaterialExpressionMultiply_3 {
   class=MaterialExpressionMultiply
   inputs [
    A {from=MaterialExpressionTime_2}
    B {from=""}
   ]
   properties {constB=0.2}
  }
  MaterialExpressionMultiply_30 {
   class=MaterialExpressionMultiply
   inputs [
    A {from=MaterialExpressionMax_2}
    B {from=""}
   ]
   properties {constB=-1}
  }
  MaterialExpressionMultiply_31 {
   class=MaterialExpressionMultiply
   inputs [
    A {from=MaterialExpressionNamedRerouteDeclaration_44}
    B {from=MaterialExpressionConstant_26}
   ]
   properties {constB=0.6667}
  }
  MaterialExpressionMultiply_4 {
   class=MaterialExpressionMultiply
   inputs [
    A {from=MaterialExpressionScalarParameter_6}
    B {from=MaterialExpressionNamedRerouteDeclaration_38}
   ]
  }
  MaterialExpressionMultiply_41 {
   class=MaterialExpressionMultiply
   inputs [
    A {from=MaterialExpressionTime_6}
    B {from=""}
   ]
   properties {constB=0.05}
  }
  MaterialExpressionMultiply_45 {
   class=MaterialExpressionMultiply
   inputs [
    A {from=MaterialExpressionMaterialFunctionCall_37, output=Result}
    B {from=MaterialExpressionScalarParameter_21}
   ]
   properties {constA=0.15, constB=0.15}
  }
  MaterialExpressionMultiply_47 {
   class=MaterialExpressionMultiply
   inputs [
    A {from=MaterialExpressionMultiply_77}
    B {from=MaterialExpressionNamedRerouteUsage_58}
   ]
  }
  MaterialExpressionMultiply_63 {
   class=MaterialExpressionMultiply
   inputs [
    A {from=MaterialExpressionConstant_40}
    B {from=""}
   ]
   properties {constB=-1}
  }
  MaterialExpressionMultiply_64 {
   class=MaterialExpressionMultiply
   inputs [
    A {from=MaterialExpressionAdd_33}
    B {from=MaterialExpressionLinearInterpolate_107}
   ]
  }
  MaterialExpressionMultiply_65 {
   class=MaterialExpressionMultiply
   inputs [
    A {from=MaterialExpressionConstant_42}
    B {from=""}
   ]
   properties {constB=-1}
  }
  MaterialExpressionMultiply_66 {
   class=MaterialExpressionMultiply
   inputs [
    A {from=MaterialExpressionSubtract_20}
    B {from=MaterialExpressionLinearInterpolate_109}
   ]
  }
  MaterialExpressionMultiply_67 {
   class=MaterialExpressionMultiply
   inputs [
    A {from=MaterialExpressionConstant_44}
    B {from=""}
   ]
   properties {constB=-1}
  }
  MaterialExpressionMultiply_68 {
   class=MaterialExpressionMultiply
   inputs [
    A {from=MaterialExpressionSubtract_17}
    B {from=MaterialExpressionConstant2Vector_48}
   ]
  }
  MaterialExpressionMultiply_69 {
   class=MaterialExpressionMultiply
   inputs [
    A {from=MaterialExpressionNamedRerouteUsage_101}
    B {from=MaterialExpressionLinearInterpolate_96}
   ]
   properties {desc=Palm Darkness}
  }
  MaterialExpressionMultiply_70 {
   class=MaterialExpressionMultiply
   inputs [
    A {from=MaterialExpressionNamedRerouteUsage_122}
    B {from=MaterialExpressionLinearInterpolate_98}
   ]
   properties {constB=0.4}
  }
  MaterialExpressionMultiply_71 {
   class=MaterialExpressionMultiply
   inputs [
    A {from=MaterialExpressionSubtract_18}
    B {from=MaterialExpressionAppendVector_5}
   ]
  }
  MaterialExpressionMultiply_72 {
   class=MaterialExpressionMultiply
   inputs [
    A {from=MaterialExpressionTime_5}
    B {from=MaterialExpressionConstant_24}
   ]
   properties {constB=2}
  }
  MaterialExpressionMultiply_73 {
   class=MaterialExpressionMultiply
   inputs [
    A {from=MaterialExpressionTextureSample_18, output=B}
    B {from=""}
   ]
   properties {constB=0.2}
  }
  MaterialExpressionMultiply_74 {
   class=MaterialExpressionMultiply
   inputs [
    A {from=MaterialExpressionMax_7}
    B {from=""}
   ]
   properties {constB=2}
  }
  MaterialExpressionMultiply_75 {
   class=MaterialExpressionMultiply
   inputs [
    A {from=MaterialExpressionMultiply_82}
    B {from=MaterialExpressionScalarParameter_20}
   ]
  }
  MaterialExpressionMultiply_76 {
   class=MaterialExpressionMultiply
   inputs [
    A {from=MaterialExpressionTextureSample_6, output=R}
    B {from=MaterialExpressionConstant2Vector_50}
   ]
   properties {constA=0.2, constB=0.1}
  }
  MaterialExpressionMultiply_77 {
   class=MaterialExpressionMultiply
   inputs [
    A {from=MaterialExpressionTextureSample_19, output=RGB}
    B {from=MaterialExpressionLinearInterpolate_100}
   ]
  }
  MaterialExpressionMultiply_78 {
   class=MaterialExpressionMultiply
   inputs [
    A {from=MaterialExpressionNamedRerouteUsage_126}
    B {from=MaterialExpressionLinearInterpolate_101}
   ]
  }
  MaterialExpressionMultiply_79 {
   class=MaterialExpressionMultiply
   inputs [
    A {from=MaterialExpressionScalarParameter_18}
    B {from=""}
   ]
   properties {constB=0.8}
  }
  MaterialExpressionMultiply_80 {
   class=MaterialExpressionMultiply
   inputs [
    A {from=MaterialExpressionSubtract_21}
    B {from=MaterialExpressionMaterialFunctionCall_50, output=Lerp 4 Inputs}
   ]
  }
  MaterialExpressionMultiply_81 {
   class=MaterialExpressionMultiply
   inputs [
    A {from=MaterialExpressionMultiply_74}
    B {from=MaterialExpressionNamedRerouteUsage_46}
   ]
   properties {constB=0.01}
  }
  MaterialExpressionMultiply_82 {
   class=MaterialExpressionMultiply
   inputs [
    A {from=MaterialExpressionMultiply_81}
    B {from=""}
   ]
   properties {constB=3}
  }
  MaterialExpressionMultiply_83 {
   class=MaterialExpressionMultiply
   inputs [
    A {from=MaterialExpressionNamedRerouteUsage_125}
    B {from=MaterialExpressionLinearInterpolate_99}
   ]
  }
  MaterialExpressionMultiply_84 {
   class=MaterialExpressionMultiply
   inputs [
    A {from=MaterialExpressionSubtract_22}
    B {from=MaterialExpressionAppendVector_8}
   ]
   properties {constB=2.2}
  }
  MaterialExpressionNamedRerouteDeclaration_21 {
   class=MaterialExpressionNamedRerouteDeclaration
   inputs [None {from=MaterialExpressionMaterialFunctionCall_3, output=Scale Y}]
   properties {name=Scale Y, nodeColor=<rgba 0 0 0 1>}
  }
  MaterialExpressionNamedRerouteDeclaration_24 {
   class=MaterialExpressionNamedRerouteDeclaration
   inputs [None {from=MaterialExpressionMaterialFunctionCall_5, output=Focal Mask}]
   properties {name=Focal Mask, nodeColor=<rgba 0 0 0 1>}
  }
  MaterialExpressionNamedRerouteDeclaration_30 {
   class=MaterialExpressionNamedRerouteDeclaration
   inputs [None {from=MaterialExpressionMaterialFunctionCall_5, output=Transition Value}]
   properties {name=Transition, nodeColor=<rgba 0 0 0 1>}
  }
  MaterialExpressionNamedRerouteDeclaration_31 {
   class=MaterialExpressionNamedRerouteDeclaration
   inputs [None {from=MaterialExpressionMaterialFunctionCall_3, output=UV}]
   properties {name=UV, nodeColor=<rgba 0 0 0 1>}
  }
  MaterialExpressionNamedRerouteDeclaration_32 {
   class=MaterialExpressionNamedRerouteDeclaration
   inputs [None {from=MaterialExpressionLinearInterpolate_31}]
   properties {name=Background Color, nodeColor=<rgba 0 0 0 1>}
  }
  MaterialExpressionNamedRerouteDeclaration_33 {
   class=MaterialExpressionNamedRerouteDeclaration
   inputs [None {from=MaterialExpressionMaterialFunctionCall_5, output=Focal Point UV}]
   properties {name=Focal UV, nodeColor=<rgba 0 0 0 1>}
  }
  MaterialExpressionNamedRerouteDeclaration_34 {
   class=MaterialExpressionNamedRerouteDeclaration
   inputs [None {from=MaterialExpressionMultiply_2}]
   properties {name=Wind Distortion, nodeColor=<rgba 0 0 0 1>}
  }
  MaterialExpressionNamedRerouteDeclaration_35 {
   class=MaterialExpressionNamedRerouteDeclaration
   inputs [None {from=MaterialExpressionTextureObject_2}]
   properties {name=Palm Tree Texture, nodeColor=<rgba 0 0 0 1>}
  }
  MaterialExpressionNamedRerouteDeclaration_36 {
   class=MaterialExpressionNamedRerouteDeclaration
   inputs [None {from=MaterialExpressionMaterialFunctionCall_3, output=Scale X}]
   properties {name=Scale X, nodeColor=<rgba 0 0 0 1>}
  }
  MaterialExpressionNamedRerouteDeclaration_38 {
   class=MaterialExpressionNamedRerouteDeclaration
   inputs [None {from=MaterialExpressionMultiply_24}]
   properties {name=Wind Time Mask, nodeColor=<rgba 0 0 0 1>}
  }
  MaterialExpressionNamedRerouteDeclaration_39 {
   class=MaterialExpressionNamedRerouteDeclaration
   inputs [None {from=MaterialExpressionTextureSample_0, output=R}]
   properties {name=Panning Noise, nodeColor=<rgba 0 0 0 1>}
  }
  MaterialExpressionNamedRerouteDeclaration_40 {
   class=MaterialExpressionNamedRerouteDeclaration
   inputs [None {from=MaterialExpressionSaturate_3}]
   properties {name=Base Gradient, nodeColor=<rgba 0 0 0 1>}
  }
  MaterialExpressionNamedRerouteDeclaration_41 {
   class=MaterialExpressionNamedRerouteDeclaration
   inputs [None {from=MaterialExpressionSmoothStep_1}]
   properties {name=Smooth Focal Mask, nodeColor=<rgba 0 0 0 1>}
  }
  MaterialExpressionNamedRerouteDeclaration_42 {class=MaterialExpressionNamedRerouteDeclaration, inputs [None {from=""}], properties {name=Color Overlay, nodeColor=<rgba 0 0 0 1>}}
  MaterialExpressionNamedRerouteDeclaration_43 {
   class=MaterialExpressionNamedRerouteDeclaration
   inputs [None {from=MaterialExpressionLinearInterpolate_42}]
   properties {name=Overlay Opacity, nodeColor=<rgba 0 0 0 1>}
  }
  MaterialExpressionNamedRerouteDeclaration_44 {
   class=MaterialExpressionNamedRerouteDeclaration
   inputs [None {from=MaterialExpressionSubtract_10}]
   properties {name=InitialScaleHack, nodeColor=<rgba 0 0 0 1>}
  }
  MaterialExpressionNamedRerouteDeclaration_64 {
   class=MaterialExpressionNamedRerouteDeclaration
   inputs [None {from=MaterialExpressionMultiply_45}]
   properties {name=Grain Noise, nodeColor=<rgba 0 0 0 1>}
  }
  MaterialExpressionNamedRerouteDeclaration_65 {
   class=MaterialExpressionNamedRerouteDeclaration
   inputs [None {from=MaterialExpressionTextureSample_15, output=A}]
   properties {name=Palm Set 01 Mask, nodeColor=<rgba 0 0 0 1>}
  }
  MaterialExpressionNamedRerouteDeclaration_66 {
   class=MaterialExpressionNamedRerouteDeclaration
   inputs [None {from=MaterialExpressionLinearInterpolate_97}]
   properties {name=Palm Blending, nodeColor=<rgba 0 0 0 1>}
  }
  MaterialExpressionNamedRerouteDeclaration_67 {
   class=MaterialExpressionNamedRerouteDeclaration
   inputs [None {from=MaterialExpressionTextureSample_15, output=R}]
   properties {name=Palm Set 01 Details, nodeColor=<rgba 0 0 0 1>}
  }
  MaterialExpressionNamedRerouteDeclaration_68 {
   class=MaterialExpressionNamedRerouteDeclaration
   inputs [None {from=MaterialExpressionMaterialFunctionCall_41, output=Twist UV}]
   properties {name=Palm UV, nodeColor=<rgba 0 0 0 1>}
  }
  MaterialExpressionNamedRerouteDeclaration_69 {
   class=MaterialExpressionNamedRerouteDeclaration
   inputs [None {from=MaterialExpressionTextureSample_16, output=A}]
   properties {name=Palm Set 02 Mask, nodeColor=<rgba 0 0 0 1>}
  }
  MaterialExpressionNamedRerouteDeclaration_70 {
   class=MaterialExpressionNamedRerouteDeclaration
   inputs [None {from=MaterialExpressionTextureSample_16, output=R}]
   properties {name=Palm Set 02 Details, nodeColor=<rgba 0 0 0 1>}
  }
  MaterialExpressionNamedRerouteDeclaration_71 {
   class=MaterialExpressionNamedRerouteDeclaration
   inputs [None {from=MaterialExpressionMaterialFunctionCall_40, output=Twist UV}]
   properties {name=Palm UV Set 02, nodeColor=<rgba 0 0 0 1>}
  }
  MaterialExpressionNamedRerouteDeclaration_72 {
   class=MaterialExpressionNamedRerouteDeclaration
   inputs [None {from=MaterialExpressionLinearInterpolate_103}]
   properties {name=Palm Blur, nodeColor=<rgba 0 0 0 1>}
  }
  MaterialExpressionNamedRerouteDeclaration_73 {
   class=MaterialExpressionNamedRerouteDeclaration
   inputs [None {from=MaterialExpressionRound_2}]
   properties {name=Mirror Mask, nodeColor=<rgba 0 0 0 1>}
  }
  MaterialExpressionNamedRerouteDeclaration_74 {
   class=MaterialExpressionNamedRerouteDeclaration
   inputs [None {from=MaterialExpressionTextureSample_17, output=A}]
   properties {name=Palm Set 03 Mask, nodeColor=<rgba 0 0 0 1>}
  }
  MaterialExpressionNamedRerouteDeclaration_75 {
   class=MaterialExpressionNamedRerouteDeclaration
   inputs [None {from=MaterialExpressionTextureSample_17, output=R}]
   properties {name=Palm Set 03 Details, nodeColor=<rgba 0 0 0 1>}
  }
  MaterialExpressionNamedRerouteDeclaration_76 {
   class=MaterialExpressionNamedRerouteDeclaration
   inputs [None {from=MaterialExpressionRotator_19}]
   properties {name=Palm UV Set 02_1, nodeColor=<rgba 0 0 0 1>}
  }
  MaterialExpressionNamedRerouteDeclaration_77 {
   class=MaterialExpressionNamedRerouteDeclaration
   inputs [None {from=MaterialExpressionMaterialFunctionCall_42, output=UVs}]
   properties {name=Palm Scale Y, nodeColor=<rgba 0 0 0 1>}
  }
  MaterialExpressionNamedRerouteDeclaration_78 {
   class=MaterialExpressionNamedRerouteDeclaration
   inputs [None {from=MaterialExpressionMultiply_70}]
   properties {name=Vignette Blend, nodeColor=<rgba 0 0 0 1>}
  }
  MaterialExpressionNamedRerouteDeclaration_79 {
   class=MaterialExpressionNamedRerouteDeclaration
   inputs [None {from=MaterialExpressionMultiply_83}]
   properties {name=Film Grain Blend, nodeColor=<rgba 0 0 0 1>}
  }
  MaterialExpressionNamedRerouteDeclaration_80 {
   class=MaterialExpressionNamedRerouteDeclaration
   inputs [None {from=MaterialExpressionMultiply_47}]
   properties {name=Light Leaks, nodeColor=<rgba 0 0 0 1>}
  }
  MaterialExpressionNamedRerouteDeclaration_81 {
   class=MaterialExpressionNamedRerouteDeclaration
   inputs [None {from=MaterialExpressionSaturate_12}]
   properties {name=Light Leaks Blend, nodeColor=<rgba 0 0 0 1>}
  }
  MaterialExpressionNamedRerouteDeclaration_82 {
   class=MaterialExpressionNamedRerouteDeclaration
   inputs [None {from=MaterialExpressionMaterialFunctionCall_44, output=Stepped Gradient}]
   properties {name=Film Step Time, nodeColor=<rgba 0 0 0 1>}
  }
  MaterialExpressionNamedRerouteUsage_100 {class=MaterialExpressionNamedRerouteUsage}
  MaterialExpressionNamedRerouteUsage_101 {class=MaterialExpressionNamedRerouteUsage}
  MaterialExpressionNamedRerouteUsage_102 {class=MaterialExpressionNamedRerouteUsage}
  MaterialExpressionNamedRerouteUsage_103 {class=MaterialExpressionNamedRerouteUsage}
  MaterialExpressionNamedRerouteUsage_104 {class=MaterialExpressionNamedRerouteUsage}
  MaterialExpressionNamedRerouteUsage_105 {class=MaterialExpressionNamedRerouteUsage}
  MaterialExpressionNamedRerouteUsage_106 {class=MaterialExpressionNamedRerouteUsage}
  MaterialExpressionNamedRerouteUsage_107 {class=MaterialExpressionNamedRerouteUsage}
  MaterialExpressionNamedRerouteUsage_108 {class=MaterialExpressionNamedRerouteUsage}
  MaterialExpressionNamedRerouteUsage_109 {class=MaterialExpressionNamedRerouteUsage}
  MaterialExpressionNamedRerouteUsage_110 {class=MaterialExpressionNamedRerouteUsage}
  MaterialExpressionNamedRerouteUsage_111 {class=MaterialExpressionNamedRerouteUsage}
  MaterialExpressionNamedRerouteUsage_112 {class=MaterialExpressionNamedRerouteUsage}
  MaterialExpressionNamedRerouteUsage_113 {class=MaterialExpressionNamedRerouteUsage}
  MaterialExpressionNamedRerouteUsage_114 {class=MaterialExpressionNamedRerouteUsage}
  MaterialExpressionNamedRerouteUsage_115 {class=MaterialExpressionNamedRerouteUsage}
  MaterialExpressionNamedRerouteUsage_116 {class=MaterialExpressionNamedRerouteUsage}
  MaterialExpressionNamedRerouteUsage_117 {class=MaterialExpressionNamedRerouteUsage}
  MaterialExpressionNamedRerouteUsage_118 {class=MaterialExpressionNamedRerouteUsage}
  MaterialExpressionNamedRerouteUsage_119 {class=MaterialExpressionNamedRerouteUsage}
  MaterialExpressionNamedRerouteUsage_120 {class=MaterialExpressionNamedRerouteUsage}
  MaterialExpressionNamedRerouteUsage_121 {class=MaterialExpressionNamedRerouteUsage}
  MaterialExpressionNamedRerouteUsage_122 {class=MaterialExpressionNamedRerouteUsage}
  MaterialExpressionNamedRerouteUsage_123 {class=MaterialExpressionNamedRerouteUsage}
  MaterialExpressionNamedRerouteUsage_124 {class=MaterialExpressionNamedRerouteUsage}
  MaterialExpressionNamedRerouteUsage_125 {class=MaterialExpressionNamedRerouteUsage}
  MaterialExpressionNamedRerouteUsage_126 {class=MaterialExpressionNamedRerouteUsage}
  MaterialExpressionNamedRerouteUsage_127 {class=MaterialExpressionNamedRerouteUsage}
  MaterialExpressionNamedRerouteUsage_128 {class=MaterialExpressionNamedRerouteUsage}
  MaterialExpressionNamedRerouteUsage_129 {class=MaterialExpressionNamedRerouteUsage}
  MaterialExpressionNamedRerouteUsage_130 {class=MaterialExpressionNamedRerouteUsage}
  MaterialExpressionNamedRerouteUsage_131 {class=MaterialExpressionNamedRerouteUsage}
  MaterialExpressionNamedRerouteUsage_132 {class=MaterialExpressionNamedRerouteUsage}
  MaterialExpressionNamedRerouteUsage_133 {class=MaterialExpressionNamedRerouteUsage}
  MaterialExpressionNamedRerouteUsage_134 {class=MaterialExpressionNamedRerouteUsage}
  MaterialExpressionNamedRerouteUsage_135 {class=MaterialExpressionNamedRerouteUsage}
  MaterialExpressionNamedRerouteUsage_136 {class=MaterialExpressionNamedRerouteUsage}
  MaterialExpressionNamedRerouteUsage_137 {class=MaterialExpressionNamedRerouteUsage}
  MaterialExpressionNamedRerouteUsage_138 {class=MaterialExpressionNamedRerouteUsage}
  MaterialExpressionNamedRerouteUsage_139 {class=MaterialExpressionNamedRerouteUsage}
  MaterialExpressionNamedRerouteUsage_140 {class=MaterialExpressionNamedRerouteUsage}
  MaterialExpressionNamedRerouteUsage_141 {class=MaterialExpressionNamedRerouteUsage}
  MaterialExpressionNamedRerouteUsage_3 {class=MaterialExpressionNamedRerouteUsage}
  MaterialExpressionNamedRerouteUsage_30 {class=MaterialExpressionNamedRerouteUsage}
  MaterialExpressionNamedRerouteUsage_37 {class=MaterialExpressionNamedRerouteUsage}
  MaterialExpressionNamedRerouteUsage_38 {class=MaterialExpressionNamedRerouteUsage}
  MaterialExpressionNamedRerouteUsage_4 {class=MaterialExpressionNamedRerouteUsage}
  MaterialExpressionNamedRerouteUsage_42 {class=MaterialExpressionNamedRerouteUsage}
  MaterialExpressionNamedRerouteUsage_43 {class=MaterialExpressionNamedRerouteUsage}
  MaterialExpressionNamedRerouteUsage_44 {class=MaterialExpressionNamedRerouteUsage}
  MaterialExpressionNamedRerouteUsage_46 {class=MaterialExpressionNamedRerouteUsage}
  MaterialExpressionNamedRerouteUsage_58 {class=MaterialExpressionNamedRerouteUsage}
  MaterialExpressionNamedRerouteUsage_60 {class=MaterialExpressionNamedRerouteUsage}
  MaterialExpressionNamedRerouteUsage_7 {class=MaterialExpressionNamedRerouteUsage}
  MaterialExpressionNamedRerouteUsage_9 {class=MaterialExpressionNamedRerouteUsage}
  MaterialExpressionOneMinus_1 {class=MaterialExpressionOneMinus, inputs [None {from=MaterialExpressionLength_0}]}
  MaterialExpressionOneMinus_2 {class=MaterialExpressionOneMinus, inputs [None {from=MaterialExpressionLength_2}]}
  MaterialExpressionOneMinus_4 {class=MaterialExpressionOneMinus, inputs [None {from=MaterialExpressionMaterialFunctionCall_43, output=Smooth}]}
  MaterialExpressionPanner_0 {
   class=MaterialExpressionPanner
   inputs [
    Coordinate {from=MaterialExpressionMaterialFunctionCall_6, output=UVs}
    Time {from=""}
    Speed {from=""}
   ]
   properties {speedX=0.1}
  }
  MaterialExpressionPower_4 {
   class=MaterialExpressionPower
   inputs [
    Base {from=MaterialExpressionSaturate_14}
    Exp {from=""}
   ]
   properties {constExponent=4}
  }
  MaterialExpressionPower_5 {
   class=MaterialExpressionPower
   inputs [
    Base {from=MaterialExpressionOneMinus_4}
    Exp {from=MaterialExpressionConstant_47}
   ]
   properties {constExponent=8}
  }
  MaterialExpressionReroute_0 {class=MaterialExpressionReroute, inputs [None {from=MaterialExpressionConstant3Vector_4}]}
  MaterialExpressionReroute_1 {class=MaterialExpressionReroute, inputs [None {from=MaterialExpressionTextureObject_7}]}
  MaterialExpressionReroute_10 {class=MaterialExpressionReroute, inputs [None {from=MaterialExpressionNamedRerouteDeclaration_68}]}
  MaterialExpressionReroute_11 {class=MaterialExpressionReroute, inputs [None {from=MaterialExpressionNamedRerouteDeclaration_71}]}
  MaterialExpressionReroute_12 {class=MaterialExpressionReroute, inputs [None {from=MaterialExpressionNamedRerouteDeclaration_76}]}
  MaterialExpressionRotator_14 {
   class=MaterialExpressionRotator
   inputs [
    Coordinate {from=MaterialExpressionReroute_10, output=""}
    Time {from=MaterialExpressionLinearInterpolate_84, output=""}
   ]
   properties {centerY=1, speed=6.2832}
  }
  MaterialExpressionRotator_15 {
   class=MaterialExpressionRotator
   inputs [
    Coordinate {from=MaterialExpressionMultiply_64, output=""}
    Time {from=MaterialExpressionConstant_41, output=""}
   ]
   properties {centerY=1, speed=6.2832}
  }
  MaterialExpressionRotator_16 {
   class=MaterialExpressionRotator
   inputs [
    Coordinate {from=MaterialExpressionReroute_11, output=""}
    Time {from=MaterialExpressionLinearInterpolate_87, output=""}
   ]
   properties {centerY=1, speed=6.2832}
  }
  MaterialExpressionRotator_17 {
   class=MaterialExpressionRotator
   inputs [
    Coordinate {from=MaterialExpressionAdd_43, output=""}
    Time {from=MaterialExpressionConstant_43, output=""}
   ]
   properties {centerY=1, speed=6.2832}
  }
  MaterialExpressionRotator_18 {
   class=MaterialExpressionRotator
   inputs [
    Coordinate {from=MaterialExpressionReroute_12, output=""}
    Time {from=MaterialExpressionLinearInterpolate_92, output=""}
   ]
   properties {centerY=1, speed=6.2832}
  }
  MaterialExpressionRotator_19 {
   class=MaterialExpressionRotator
   inputs [
    Coordinate {from=MaterialExpressionAdd_38, output=""}
    Time {from=MaterialExpressionConstant_45, output=""}
   ]
   properties {centerY=1, speed=6.2832}
  }
  MaterialExpressionRotator_20 {
   class=MaterialExpressionRotator
   inputs [
    Coordinate {from=MaterialExpressionAdd_44, output=""}
    Time {from=MaterialExpressionLinearInterpolate_104, output=""}
   ]
   properties {centerY=1, speed=6.2832}
  }
  MaterialExpressionRound_2 {class=MaterialExpressionRound, inputs [None {from=MaterialExpressionComponentMask_32, output=""}], properties {}}
  MaterialExpressionSaturate_0 {class=MaterialExpressionSaturate, inputs [None {from=MaterialExpressionOneMinus_1}]}
  MaterialExpressionSaturate_11 {class=MaterialExpressionSaturate, inputs [None {from=MaterialExpressionMultiply_75}]}
  MaterialExpressionSaturate_12 {class=MaterialExpressionSaturate, inputs [None {from=MaterialExpressionMultiply_78}]}
  MaterialExpressionSaturate_13 {class=MaterialExpressionSaturate, inputs [None {from=MaterialExpressionNamedRerouteUsage_132}]}
  MaterialExpressionSaturate_14 {class=MaterialExpressionSaturate, inputs [None {from=MaterialExpressionComponentMask_35}]}
  MaterialExpressionSaturate_3 {class=MaterialExpressionSaturate, inputs [None {from=MaterialExpressionLinearInterpolate_2}]}
  MaterialExpressionSaturate_4 {class=MaterialExpressionSaturate, inputs [None {from=MaterialExpressionOneMinus_2}]}
  MaterialExpressionScalarParameter_10 {class=MaterialExpressionScalarParameter, properties {parameterName=OverrideTheme, sliderMax=1}}
  MaterialExpressionScalarParameter_11 {class=MaterialExpressionScalarParameter, properties {defaultValue=1, parameterName=OverrideTransition, sliderMax=1}}
  MaterialExpressionScalarParameter_18 {class=MaterialExpressionScalarParameter, properties {defaultValue=2, group=Palm Trees, parameterName=Palm Blur, sliderMax=4}}
  MaterialExpressionScalarParameter_19 {class=MaterialExpressionScalarParameter, properties {defaultValue=1, parameterName=Palm Zoom}}
  MaterialExpressionScalarParameter_20 {class=MaterialExpressionScalarParameter, properties {defaultValue=2, group=Film Grain, parameterName=Film Grain Intensity}}
  MaterialExpressionScalarParameter_21 {class=MaterialExpressionScalarParameter, properties {defaultValue=0.15, group=Film Grain, parameterName=Film Noise Opacity, sliderMax=1}}
  MaterialExpressionScalarParameter_22 {class=MaterialExpressionScalarParameter, properties {defaultValue=0.7, group=Palm Trees, parameterName=Palm Darkness, sliderMax=1}}
  MaterialExpressionScalarParameter_23 {class=MaterialExpressionScalarParameter, properties {defaultValue=0.9, group=Palm Trees, parameterName=Palm Burn, sliderMax=1}}
  MaterialExpressionScalarParameter_4 {class=MaterialExpressionScalarParameter, properties {defaultValue=1.5, parameterName=FocalScale}}
  MaterialExpressionScalarParameter_6 {class=MaterialExpressionScalarParameter, properties {defaultValue=0.005, group=Wind, parameterName=Wind Distortion}}
  MaterialExpressionScalarParameter_7 {class=MaterialExpressionScalarParameter, properties {defaultValue=0.225, group=Wind, parameterName=Wind Intensity}}
  MaterialExpressionScalarParameter_8 {class=MaterialExpressionScalarParameter, properties {group=UV, parameterName=UseLocalUV, sliderMax=1}}
  MaterialExpressionScalarParameter_9 {class=MaterialExpressionScalarParameter, properties {defaultValue=0.4, group=Vignette, parameterName=Vignette Opacity, sliderMax=1}}
  MaterialExpressionSine_2 {class=MaterialExpressionSine, inputs [None {from=MaterialExpressionComponentMask_12}]}
  MaterialExpressionSine_4 {class=MaterialExpressionSine, inputs [None {from=MaterialExpressionFrac_4}]}
  MaterialExpressionSmoothStep_1 {
   class=MaterialExpressionSmoothStep
   inputs [
    Min {from="", output=""}
    Max {from="", output=""}
    Value {from=MaterialExpressionMaterialFunctionCall_20, output=EaseInSine}
   ]
   properties {}
  }
  MaterialExpressionSmoothStep_3 {
   class=MaterialExpressionSmoothStep
   inputs [
    Min {from="", output=""}
    Max {from="", output=""}
    Value {from=MaterialExpressionSaturate_4, output=""}
   ]
   properties {}
  }
  MaterialExpressionSmoothStep_5 {
   class=MaterialExpressionSmoothStep
   inputs [
    Min {from=MaterialExpressionConstant_18, output=""}
    Max {from="", output=""}
    Value {from=MaterialExpressionLength_4, output=""}
   ]
   properties {constMin=0.8}
  }
  MaterialExpressionSubtract_10 {
   class=MaterialExpressionSubtract
   inputs [
    A {from=MaterialExpressionComponentMask_17}
    B {from=MaterialExpressionConstant_7}
   ]
  }
  MaterialExpressionSubtract_17 {
   class=MaterialExpressionSubtract
   inputs [
    A {from=MaterialExpressionAdd_37}
    B {from=MaterialExpressionConstant2Vector_49}
   ]
  }
  MaterialExpressionSubtract_18 {
   class=MaterialExpressionSubtract
   inputs [
    A {from=MaterialExpressionNamedRerouteUsage_123}
    B {from=""}
   ]
   properties {constB=0.5}
  }
  MaterialExpressionSubtract_19 {
   class=MaterialExpressionSubtract
   inputs [
    A {from=MaterialExpressionNamedRerouteUsage_128}
    B {from=""}
   ]
   properties {constB=0.5}
  }
  MaterialExpressionSubtract_2 {
   class=MaterialExpressionSubtract
   inputs [
    A {from=MaterialExpressionNamedRerouteUsage_4}
    B {from=MaterialExpressionAppendVector_2}
   ]
  }
  MaterialExpressionSubtract_20 {
   class=MaterialExpressionSubtract
   inputs [
    A {from=MaterialExpressionAdd_35}
    B {from=MaterialExpressionConstant2Vector_58}
   ]
  }
  MaterialExpressionSubtract_21 {
   class=MaterialExpressionSubtract
   inputs [
    A {from=MaterialExpressionNamedRerouteUsage_124}
    B {from=""}
   ]
   properties {constB=0.5}
  }
  MaterialExpressionSubtract_22 {
   class=MaterialExpressionSubtract
   inputs [
    A {from=MaterialExpressionNamedRerouteUsage_140}
    B {from=""}
   ]
   properties {constB=0.5}
  }
  MaterialExpressionSubtract_6 {
   class=MaterialExpressionSubtract
   inputs [
    A {from=MaterialExpressionAdd_2}
    B {from=""}
   ]
   properties {constB=0.5}
  }
  MaterialExpressionSubtract_8 {
   class=MaterialExpressionSubtract
   inputs [
    A {from=MaterialExpressionNamedRerouteUsage_37}
    B {from=MaterialExpressionConstant2Vector_11}
   ]
   properties {constB=0.5}
  }
  MaterialExpressionSubtract_9 {
   class=MaterialExpressionSubtract
   inputs [
    A {from=MaterialExpressionAdd_11}
    B {from=""}
   ]
   properties {constB=0.5}
  }
  MaterialExpressionTextureObject_2 {
   class=MaterialExpressionTextureObject
   inputs []
   properties {samplerType=SAMPLERTYPE_LinearColor, texture=&/T_UI_Ch7S1_PalmTree}
  }
  MaterialExpressionTextureObject_6 {
   class=MaterialExpressionTextureObject
   inputs []
   properties {samplerType=SAMPLERTYPE_LinearColor, texture=&/T_UI_Ch7S1_FilmGrain}
  }
  MaterialExpressionTextureObject_7 {
   class=MaterialExpressionTextureObject
   inputs []
   properties {texture=&/T_UI_Ch7S1_LightLeaks}
  }
  MaterialExpressionTextureProperty_2 {class=MaterialExpressionTextureProperty, inputs [Tex {from=MaterialExpressionNamedRerouteUsage_106, output=""}], properties {}}
  MaterialExpressionTextureSample_0 {
   class=MaterialExpressionTextureSample
   inputs [
    UVs {from=MaterialExpressionPanner_0}
    Tex {from=""}
    Apply View MipBias {from=""}
   ]
   properties {samplerType=SAMPLERTYPE_Grayscale, texture=&LowResBlurredNoise}
  }
  MaterialExpressionTextureSample_15 {
   class=MaterialExpressionTextureSample
   inputs [
    UVs {from=MaterialExpressionAdd_32}
    Tex {from=MaterialExpressionNamedRerouteUsage_107}
    Level {from=MaterialExpressionNamedRerouteDeclaration_72}
    Apply View MipBias {from=""}
   ]
   properties {mipValueMode=TMVM_MipLevel}
  }
  MaterialExpressionTextureSample_16 {
   class=MaterialExpressionTextureSample
   inputs [
    UVs {from=MaterialExpressionAdd_34}
    Tex {from=MaterialExpressionNamedRerouteUsage_110}
    Level {from=MaterialExpressionNamedRerouteUsage_111}
    Apply View MipBias {from=""}
   ]
   properties {mipValueMode=TMVM_MipLevel}
  }
  MaterialExpressionTextureSample_17 {
   class=MaterialExpressionTextureSample
   inputs [
    UVs {from=MaterialExpressionAdd_36}
    Tex {from=MaterialExpressionNamedRerouteUsage_116}
    Level {from=MaterialExpressionNamedRerouteUsage_117}
    Apply View MipBias {from=""}
   ]
   properties {mipValueMode=TMVM_MipLevel}
  }
  MaterialExpressionTextureSample_18 {
   class=MaterialExpressionTextureSample
   inputs [
    UVs {from=MaterialExpressionAdd_40}
    Tex {from=MaterialExpressionTextureObject_6}
    Apply View MipBias {from=""}
   ]
  }
  MaterialExpressionTextureSample_19 {
   class=MaterialExpressionTextureSample
   inputs [
    UVs {from=MaterialExpressionAdd_41}
    Tex {from=MaterialExpressionTextureObject_7}
    Level {from=MaterialExpressionConstant_51}
    Apply View MipBias {from=""}
   ]
   properties {mipValueMode=TMVM_MipLevel}
  }
  MaterialExpressionTextureSample_20 {
   class=MaterialExpressionTextureSample
   inputs [
    UVs {from=MaterialExpressionAdd_14}
    Tex {from=MaterialExpressionReroute_1}
    Level {from=MaterialExpressionConstant_51}
    Apply View MipBias {from=""}
   ]
   properties {mipValueMode=TMVM_MipLevel}
  }
  MaterialExpressionTextureSample_6 {
   class=MaterialExpressionTextureSample
   inputs [
    UVs {from=MaterialExpressionMaterialFunctionCall_49, output=Perspective UVs}
    Tex {from=""}
    Apply View MipBias {from=""}
   ]
   properties {samplerType=SAMPLERTYPE_Grayscale, texture=&LowResBlurredNoise}
  }
  MaterialExpressionTime_2 {class=MaterialExpressionTime}
  MaterialExpressionTime_5 {class=MaterialExpressionTime}
  MaterialExpressionTime_6 {class=MaterialExpressionTime}
  MaterialExpressionViewSize_0 {class=MaterialExpressionViewSize, inputs [], properties {}}
 }
 properties {previewMaterialDomain=MD_UI}
}
