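# UI material for the background of a Chapter 7 Season 1 (CH7S1) upsell screen, filed with the quest UI materials.
# It is a translucent User Interface domain material whose only node is a call to the MF_UI_CH7_BackgroundForUpsell material function.
# The material adds no logic or parameters of its own, so any background visuals come from that function.
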Material /FortniteCoreUI/UI/Quests/Materials/CH7_materials/M_UI_CH7S1_Background_UI_Upsell
graph {
 properties {
  bAutomaticallySetUsageInEditor=false
  bUsedWithStaticMesh=false
  blendMode=BLEND_Translucent
  materialDomain=MD_UI
  usedShadingModels=Default Lit
 }
 nodes [MaterialFunctionCall() {materialFunction=&/MF_UI_CH7_BackgroundForUpsell}]
}
